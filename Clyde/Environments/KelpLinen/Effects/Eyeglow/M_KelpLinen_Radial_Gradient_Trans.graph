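# Translucent unlit particle material that draws a soft radial glow for the KelpLinen eyeglow effect.
# Emissive blends a parameter color toward particle color, scaled by time of day.
# Opacity is a UV sphere mask with a dynamic-parameter power, multiplied by particle alpha, High/Epic-only depth and near-camera fades, optional angle and camera-depth fades, and a KelpLinen location fade.
# The resulting material attributes pass through the Kelp audio-reactive function.
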Material /Clyde/Environments/KelpLinen/Effects/Eyeglow/M_KelpLinen_Radial_Gradient_Trans
graph {
 properties {
  bUseMaterialAttributes=true
  bUsedWithMeshParticles=true
  bUsedWithNanite=true
  bUsedWithNiagaraMeshParticles=true
  bUsedWithNiagaraSprites=true
  bUsedWithParticleSprites=true
  blendMode=BLEND_Translucent
  pixelDepthOffsetMode=PDOM_Legacy
  shadingModel=MSM_Unlit
  usedShadingModels=Unlit
 }
 nodes [
  n1 = ParticleColor()
  n2 = Multiply(
   A=LinearInterpolate(
    A=VectorParameter() {parameterName=EmissiveColor, defaultValue=<rgba 0 1 0.7481 1>}.RGB
    B=@n1.RGB
    Alpha=ScalarParameter() {defaultValue=1, parameterName=Use_ParticleColor}
   )
   B=MaterialFunctionCall(
    Value_Morning=ScalarParameter() {defaultValue=1, parameterName="Value_Morning (S)"}
    Value_Day=ScalarParameter() {defaultValue=1, parameterName="Value_Day (S)"}
    Value_Evening=ScalarParameter() {defaultValue=1, parameterName="Value_Evening (S)"}
    Value_Night=ScalarParameter() {defaultValue=1, parameterName="Value_Night (S)"}
   ) {materialFunction=&Scalar_Time_of_Day}.Result
  )
  n3 = SphereMask(
   A=TextureCoordinate()
   B=Constant() {r=0.5}
   Radius=ScalarParameter() {defaultValue=0.5, parameterName=Radius}
   Hardness=ScalarParameter() {parameterName=Hardness}
  ) {hardnessPercent=0}
  n4 = DynamicParameter() {defaultValue=<rgba 3 1 1 1>, paramNames [Power, DepthFadeDistMult, Param3, Param4]}
  n5 = Clamp(None=@n4.Power) {clampMode=CMODE_ClampMin, maxDefault=100, minDefault=1}
  n6 = Constant() {r=1}
  n7 = DepthFade(
   Opacity=MaterialFunctionCall(Distance=ScalarParameter() {defaultValue=100, parameterName=NearCameraFadeDist}) {materialFunction=&Near_Camera_Fade}.Result
   FadeDistance=Multiply(A=@n4.DepthFadeDistMult, B=ScalarParameter() {defaultValue=100, parameterName=DepthFadeDist})
  )
  n8 = QualitySwitch(Default=@n6, Low=@n6, High=@n7, Medium=@n6, Epic=@n7)
  n9 = ScalarParameter() {defaultValue=8, parameterName=AngleFade}
  n10 = Multiply(
   A=Multiply(A=Power(Base=@n3, Exp=@n5) {constExponent=3}, B=Multiply(A=@n1.A, B=@n8))
   B=StaticSwitchParameter(
    True=MaterialFunctionCall("FadeSensitivity (optional)"=@n9) {materialFunction=&FadeBasedOnViewAngle}.Result
    False=Constant() {r=1}
   ) {parameterName=Use AngleFade}
  )
  n11 = ScalarParameter() {defaultValue=100, parameterName=CameraDepthFade_Length}
  n12 = ScalarParameter() {defaultValue=110, parameterName=CameraDepthFade_Offset}
  n13 = LinearInterpolate(
   A=@n10
   B=Multiply(
    A=@n10
    B=MaterialFunctionCall(Fade Length=@n11, Fade Offset=@n12) {materialFunction=&CameraDepthFade}.Result
   )
   Alpha=ScalarParameter() {parameterName=Switch_CameraDepthFade}
  )
  n14 = MaterialFunctionCall() {materialFunction=/ClydeCosmetics/Environments/KelpLinen/MF_KelpLinen_LocationFade}
  MaterialFunctionCall!(In=SetMaterialAttributes!(
    MaterialAttributes
    Emissive Color=MaterialFunctionCall(Emissive=@n2) {materialFunction=&MF_FX_EmissiveAdjustments}.Result
    Opacity=Multiply(A=@n13, B=@n14.Output)
    World Position Offset
   ) {attributeSetTypes [#1, #2, #3]}) {materialFunction=/ClydeCosmetics/Environments/KelpLinen/MF_Kelp_AudioReactive}
 ]
}
