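# Niagara sparkle effect for the KelpLinen character, scattering huge HDR cyan-to-magenta sprites through a very large sphere, drifting with curl noise.
# System.Phase indexes float arrays for spawn rate and sprite size, so phase 0 spawns nothing and later phases intensify.
# Two near-identical looping emitters run together: a GPU one and a lower-rate one without GPU simulation.
# Each has two sprite renderers and computes enlarged, low-alpha FlareScale and FlareColor attributes.
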NiagaraSystem /Clyde/Environments/KelpLinen/Effects/Eyeglow/NS_KelpLinen_Character_Sparkles
dependencies {
 usedDataInterfaces [&NiagaraDataInterfaceCurve, &NiagaraDataInterfaceArrayFloat]
 usedDynamicInputs []
 usedModules [
  &SystemState
  &self:SystemSpawnScript.SystemScriptSource.SystemScriptGraph.NiagaraNodeAssignment_0.SetVariables_#1
 ]
}
emitters {
 Sparkles {
  simTarget=GPUComputeSim
  emitterUpdateScript [
   &EmitterState(
    Delay First Loop Only=true
    Life Cycle Mode=::NewEnumerator1
    Loop Delay=0.4253
    Loop Duration=0.6522
    Recalculate Duration Each Loop=true
    UseLoopDelay=true
   )
   &SpawnRate(
    SpawnRate=&SelectFloatFromArray(
     Direct Array Index=&MakeIntFromFloat(Float=$System.Phase)
     Float Selection Array=DataInterface() {propertyValues {FloatData [0,10,20,100]}}
    )
    Use Spawn Probability=false
   )
  ]
  particleSpawnScript [
   &InitializeParticle(
    Color=&RandomRangeLinearColor(Minimum=<rgba 0 12.2879 15 1>, Maximum=<rgba 9.2813 0 15 1>)
    Lifetime=&RandomRangeFloat(Minimum=1, Maximum=2)
    Mesh Scale Mode=::NewEnumerator3
    Sprite Rotation Mode=::NewEnumerator2
    Sprite Size Mode=::NewEnumerator3
    Uniform Sprite Size=&RandomRangeFloat(
     Minimum=&SelectFloatFromArray(
      Direct Array Index=&MakeIntFromFloat(Float=$System.Phase)
      Float Selection Array=DataInterface() {propertyValues {FloatData [0,5000,5000,10000]}}
     )
     Maximum=&SelectFloatFromArray(
      Direct Array Index=&MakeIntFromFloat(Float=$System.Phase)
      Float Selection Array=DataInterface() {propertyValues {FloatData [0,15000,15000,20000]}}
     )
    )
    UsePositionOffset=false
    Write Material Random=true
   )
   &ShapeLocation(MaskSpawnGroup=false, Override Seed=false, Sphere Radius=400000, UseOwnerScale=true)
   &CameraOffset(Camera Offset Amount=50000)
   &AddVelocity(
    Module.Use Falloff Distance=false
    Origin Offset=<0 0 20000>
    Origin Offset Coordinate Space=::World
    Override Seed=false
    Velocity Mode=::NewEnumerator1
    Velocity Origin=&Add_Vector(A=<-818307.875 -863074.375 0>, B=<0 0 0>)
    Velocity Speed=&RandomRangeFloat(Minimum=2500, Maximum=5000)
   )
  ]
  particleUpdateScript [
   &ParticleState()
   &ScaleSpriteSize(Scale Factor=&Vector2DFromFloat(Value=&FloatFromCurve(FloatCurve=DataInterface() {propertyValues {Curve {keys [<at 0 1 RCIM_Cubic>, {arriveTangent=-1.6315, interpMode=RCIM_Cubic, leaveTangent=-1.6315, tangentMode=RCTM_User, time=1, value=0}]}}}))) {module=ScaleSpriteSize001}
   &ScaleSpriteSize(Uniform Curve Sprite Scale=DataInterface() {propertyValues {Curve {keys [{tangentMode=RCTM_User, time=0, value=0}, {tangentMode=RCTM_User, time=0.65, value=1}, <at 1 0>]}}})
   &CurlNoiseForce(Noise Strength=10000, Pan Noise=false)
   &SolveForcesAndVelocity(Clamp Velocity=false, Limit Acceleration=false) {module=SolveForcesAndVelocity001}
   &ScaleColor(
    Scale Alpha=&FloatFromCurve(FloatCurve=DataInterface() {propertyValues {Curve {keys [<at 0 1>,<at 0.1 3>,<at 0.2 1>]}}})
    Scale RGB=&MakeVectorFromLinearColorRGB(Color=<rgba 0 0 0 1>)
    ScaleA=true
    ScaleRGB=false
   )
   &DynamicMaterialParameters(
    Index 0 Param 1=250
    Module.Param0WriteEnabled=true
    Module.Param1WriteEnabled=true
    Module.Param2WriteEnabled=true
    Module.Param3WriteEnabled=true
   )
   /SetVariables(Particles.FlareScale:Vector2f=&Multiply_Vector2D_ByFloat(Float=10, Vector2D=$Particles.SpriteSize)) {isSetParameters=true, module=SetVariables_#2}
   /SetVariables(Particles.FlareColor=&MultiplyLinearColor(
     Linear Color=$Particles.Color
     Scale Factor=/Niagara/DynamicInputs/UniformRange/V2/RandomRangeVector4(
      Minimum:Vector4f=<1 1 1 0.01>
      Maximum:Vector4f=<1 1 1 0.02>
      Evaluation Type:ENiagaraRandomnessEvaluation=::NewEnumerator0
      Recalculate Random Each Loop:NiagaraBool=true
      Randomness Mode:ENiagaraRandomnessMode=::NewEnumerator0
      Random Seed:NiagaraInt32=0
      Fixed Random Seed:NiagaraBool=false
     )
    )) {isSetParameters=true, module=SetVariables_#3}
  ]
  rendererClasses [&NiagaraSpriteRendererProperties]
  renderers [
   {rendererClass=&NiagaraSpriteRendererProperties}
   {rendererClass=&NiagaraSpriteRendererProperties, rendererIndex=1}
  ]
 }
 Sparkles001 {
  emitterUpdateScript [
   &EmitterState(
    Delay First Loop Only=true
    Life Cycle Mode=::NewEnumerator1
    Loop Delay=0.4253
    Loop Duration=0.6522
    Recalculate Duration Each Loop=true
    UseLoopDelay=true
   )
   &SpawnRate(
    SpawnRate=&SelectFloatFromArray(
     Direct Array Index=&MakeIntFromFloat(Float=$System.Phase)
     Float Selection Array=DataInterface() {propertyValues {FloatData [0,5,10,20]}}
    )
    Use Spawn Probability=false
   )
  ]
  particleSpawnScript [
   &InitializeParticle(
    Color=&RandomRangeLinearColor(Minimum=<rgba 0 12.2879 15 1>, Maximum=<rgba 9.2813 0 15 1>)
    Lifetime=&RandomRangeFloat(Minimum=1, Maximum=2)
    Mesh Scale Mode=::NewEnumerator3
    Sprite Rotation Mode=::NewEnumerator2
    Sprite Size Mode=::NewEnumerator3
    Uniform Sprite Size=&RandomRangeFloat(
     Minimum=&SelectFloatFromArray(
      Direct Array Index=&MakeIntFromFloat(Float=$System.Phase)
      Float Selection Array=DataInterface() {propertyValues {FloatData [0,5000,5000,10000]}}
     )
     Maximum=&SelectFloatFromArray(
      Direct Array Index=&MakeIntFromFloat(Float=$System.Phase)
      Float Selection Array=DataInterface() {propertyValues {FloatData [0,15000,15000,20000]}}
     )
    )
    UsePositionOffset=false
    Write Material Random=true
   )
   &ShapeLocation(MaskSpawnGroup=false, Override Seed=false, Sphere Radius=400000, UseOwnerScale=true)
   &CameraOffset(Camera Offset Amount=50000)
   &AddVelocity(
    Module.Use Falloff Distance=false
    Origin Offset=<0 0 20000>
    Origin Offset Coordinate Space=::World
    Override Seed=false
    Velocity Mode=::NewEnumerator1
    Velocity Origin=&Add_Vector(A=<-818307.875 -863074.375 0>, B=<0 0 0>)
    Velocity Speed=&RandomRangeFloat(Minimum=2500, Maximum=5000)
   )
  ]
  particleUpdateScript [
   &ParticleState()
   &ScaleSpriteSize(Scale Factor=&Vector2DFromFloat(Value=&FloatFromCurve(FloatCurve=DataInterface() {propertyValues {Curve {keys [<at 0 1 RCIM_Cubic>, {arriveTangent=-1.6315, interpMode=RCIM_Cubic, leaveTangent=-1.6315, tangentMode=RCTM_User, time=1, value=0}]}}}))) {module=ScaleSpriteSize001}
   &ScaleSpriteSize(Uniform Curve Sprite Scale=DataInterface() {propertyValues {Curve {keys [{tangentMode=RCTM_User, time=0, value=0}, {tangentMode=RCTM_User, time=0.65, value=1}, <at 1 0>]}}})
   &CurlNoiseForce(Noise Strength=10000, Pan Noise=false)
   &SolveForcesAndVelocity(Clamp Velocity=false, Limit Acceleration=false) {module=SolveForcesAndVelocity001}
   &ScaleColor(
    Scale Alpha=&FloatFromCurve(FloatCurve=DataInterface() {propertyValues {Curve {keys [<at 0 1>,<at 0.1 3>,<at 0.2 1>]}}})
    Scale RGB=&MakeVectorFromLinearColorRGB(Color=<rgba 0 0 0 1>)
    ScaleA=true
    ScaleRGB=false
   )
   &DynamicMaterialParameters(
    Index 0 Param 1=250
    Module.Param0WriteEnabled=true
    Module.Param1WriteEnabled=true
    Module.Param2WriteEnabled=true
    Module.Param3WriteEnabled=true
   )
   /SetVariables(Particles.FlareScale:Vector2f=&Multiply_Vector2D_ByFloat(Float=10, Vector2D=$Particles.SpriteSize)) {isSetParameters=true, module=SetVariables_#2}
   /SetVariables(Particles.FlareColor=&MultiplyLinearColor(
     Linear Color=$Particles.Color
     Scale Factor=&RandomRangeVector4(
      Minimum:Vector4f=<1 1 1 0.01>
      Maximum:Vector4f=<1 1 1 0.02>
      Evaluation Type:ENiagaraRandomnessEvaluation=::NewEnumerator0
      Recalculate Random Each Loop:NiagaraBool=true
      Randomness Mode:ENiagaraRandomnessMode=::NewEnumerator0
      Random Seed:NiagaraInt32=0
      Fixed Random Seed:NiagaraBool=false
     )
    )) {isSetParameters=true, module=SetVariables_#3}
  ]
  rendererClasses [&NiagaraSpriteRendererProperties]
  renderers [
   {rendererClass=&NiagaraSpriteRendererProperties}
   {rendererClass=&NiagaraSpriteRendererProperties, rendererIndex=1}
  ]
 }
}
summary {
 systemName=NS_KelpLinen_Character_Sparkles
 emitters [
  {emitterName=Sparkles, simTarget=GPUComputeSim, rendererClasses [&NiagaraSpriteRendererProperties]}
  {emitterName=Sparkles001, rendererClasses [&NiagaraSpriteRendererProperties]}
 ]
 userVariables []
}
